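; GA_Quest_LessThan_50Health is a quest gameplay ability that tracks whether the owning pawn is at or below a health threshold, named for 50 health.
; On activation it binds Check Health to the pawn's health-changed event and also runs a health check immediately.
; Check Health applies a configured GameplayEffect to the pawn at or below the threshold, and otherwise removes active effects by their granted tags.
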(event Ability|EventFailedToActivatePassiveAbility
  (params (ActorInfo :type Gameplay%20Ability%20Actor%20Info%20Structure)))

(event Ability|EventActivateAbility
  (|CallParentFunction)
  (Health|BindEventtoOnPawnHealthChanged (Variables|GAVars|Data|GetOwningPlayerPawn) (delegate Custom|Check%20Health))
  (CallFunction|CheckHealth))

(event Custom|Check%20Health
  (if (<= (FortPawn|GetHealth (Variables|GAVars|Data|GetOwningPlayerPawn)) (Variables|Default|GetHealthThreshold))
    (GameplayEffects|ApplyGameplayEffectToSelf :self (Ability|GetAbilitySystemComponent (Variables|GAVars|Data|GetOwningPlayerPawn)) :GameplayEffectClass (Variables|Default|GetGE_Class) :Level 0.0)
    (else
      (GameplayEffects|RemoveActiveEffectswithGrantedTags (Ability|GetAbilitySystemComponent (Variables|GAVars|Data|GetOwningPlayerPawn)) (Variables|Default|GetGE_Tags)))))
